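# Masked, unlit Niagara sprite material for stylized smoke puffs in the vehicle boost cosmetic.
# Dynamic parameters blend particle color toward a desaturated, brightness-scaled smoke tone, then feed an emissive adjustment function.
# Fake top lighting scales that color using a transformed puff normal dotted with Z, biased by particle alpha.
# The mask multiplies a second texture's alpha by a temporal-AA dither blended through near-camera fade, both dynamically controlled.
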Material /VehicleCosmetics/Boosts/Boost_Stylized/Material/M_StylizedSmokePuff
graph {
 properties {
  bUsedWithNiagaraSprites=true
  blendMode=BLEND_Masked
  opacityMaskClipValue=0.7
  pixelDepthOffsetMode=PDOM_Legacy
  shadingModel=MSM_Unlit
  usedShadingModels=Unlit
 }
 nodes [
  n1 = ParticleColor()
  n2 = DynamicParameter() {paramNames [transition,saturation,Brightness,Smoke]}
  n3 = Desaturation(None=@n1.RGB, Fraction=@n2.saturation)
  n4 = Saturate(None=LinearInterpolate(A=Constant() {r=0.235}, B=Multiply(A=@n3, B=@n2.Brightness), Alpha=@n2.Smoke))
  n5 = TextureSample() {samplerType=SAMPLERTYPE_Normal, texture=/DealershipSharedLibrary/FX/Textures/Flipbook/T_VCVFX_StylizedPuffNormalCom_N}
  n6 = Add(A=DotProduct(A=Transform(None=@n5.RGB), B=Constant3Vector() {constant=<rgba 0 0 1 0>}), B=@n1.A)
  n7 = Constant() {r=0.5}
  MaterialFunctionCall(Emissive=Multiply(A=LinearInterpolate(A=@n1.RGB, B=@n4, Alpha=@n2.transition), B=Multiply(A=@n6, B=@n7))) {materialFunction=&MF_FX_EmissiveAdjustments}
  n8 = TextureSample() {samplerType=SAMPLERTYPE_LinearColor, texture=/DealershipSharedLibrary/Assets/Sets/Stylizedpsplus/T_VCVFX_Boost_Stylized_StylizedPuff_N}
  n9 = DynamicParameter() {parameterIndex=3, paramNames [AlphaThreshold_dither, Near_CameraFade, "", ""]}
  Multiply(
   A=@n8.A
   B=LinearInterpolate(
    A=MaterialFunctionCall(Alpha Threshold=@n9.AlphaThreshold_dither) {materialFunction=&DitherTemporalAA_Fortnite}.Result
    Alpha=MaterialFunctionCall(Distance=@n9.Near_CameraFade) {materialFunction=&Near_Camera_Fade}.Result
   )
  )
 ]
}
